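; EQS query context that supplies Silkwalker patrol path actors to environment queries in Rift Encounters.
; ProvideActorsSet gathers every FortAthenaPatrolPath through GetAllActorsOfClass and keeps valid ones whose owned gameplay tags include Athena.AI.SpawnLocation.SilkWalker.
; The querier inputs are unused, so the set comes from a global actor search independent of who runs the query.
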(fn ProvideActorsSet
  (params (QuerierObject :type Object%20Reference) (QuerierActor :type Actor%20Object%20Reference))
  (bind _outactors (Actor|GetAllActorsOfClass "/Script/FortniteAI.FortAthenaPatrolPath"))
  (for _array_element _outactors
    (Utilities|IsValid _array_element
      (:Is%20Valid
        (if (GameplayTags|HasTag (GameplayTags|GetOwnedGameplayTags _array_element :!class /Script/GameplayTags.GameplayTagAssetInterface) "(TagName=\"Athena.AI.SpawnLocation.SilkWalker\")")
          (Utilities|Array|Add (LocalVariables|GetReturnValue) _array_element)))))
  (return (LocalVariables|GetReturnValue)))
